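# Physics asset for the skeletal mesh of the Valor glider chute, a cosmetic in the VinderTech glider set.
# It approximates the glider with a single box body on the Handles bone, offset and rotated relative to that bone.
# It defines no constraints, so there is no jointed or multi-body ragdoll setup.
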PhysicsAsset /BRCosmetics/Gadgets/Assets/VinderTech_GliderChute/Glider_Valor/Meshes/SK_Glider_Valor_PhysicsAsset
bodies {Handles [{shapeType=Box, extentX=493.8936, extentY=84.7031, extentZ=299.3809, center=<-0.1996 98.1715 7.8956>, rotation=<rot 0 -179.9295 -87.7597>}]}
constraints []
